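# Niagara system for the idle effect of the Polished Jade backpack cosmetic, with glowing meshes, sprites, a flickering light and flares.
# System timelines drive fade-ins and flare size and alpha curves, with flares placed at skeletal mesh sockets.
# Front-end flags switch between flare emitters that kill their particles on timeline progress and FE copies with that kill disabled.
# Colors use variant selectors (teal original, purple alternate) and time-of-day scaling.
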NiagaraSystem /BRCosmetics/Accessories/FORT_Backpacks/Backpack_F_MED_PolishedJade/FX/NS_Backpack_PolishedJade_Idle
dependencies {
 usedDataInterfaces [
  &NiagaraDataInterfaceArrayFloat3
  &NiagaraDataInterfaceMeshRendererInfo
  &NiagaraDataInterfaceCurve
  &NiagaraDataInterfaceArrayFloat
  &NiagaraDataInterfaceArrayColor
 ]
 usedDynamicInputs [
  &MakeCustomFloatFromBool
  &Boolean_NotOperation
  &Boolean_AndOperation
  &FN_Select_Vector_from_Time_Of_Day
 ]
 usedModules [
  &SystemState
  &Timeline
  &FN_VariantManager
  &self:SystemSpawnScript.SystemScriptSource.SystemScriptGraph.NiagaraNodeAssignment_0.SetVariables_#1
  &self:SystemSpawnScript.SystemScriptSource.SystemScriptGraph.NiagaraNodeAssignment_1.SetVariables_#2
 ]
}
emitters {
 BlowingParticles {
  emitterUpdateScript [
   &EmitterState(
    Inactive Response=::NewEnumerator1
    Life Cycle Mode=::NewEnumerator1
    Loop Behavior=::NewEnumerator1
    Loop Duration Mode=::NewEnumerator1
    UseLoopDelay=false
   )
   &SpawnBurst_Instantaneous(Spawn Count=8, Use Spawn Probability=false, UseLoopCountLimit=false)
   &SpawnRate(SpawnRate=15, Use Spawn Probability=false) {enabled=false}
   /SetVariables(
    Emitter.Color_01:LinearColor=&MultiplyLinearColor(
     Linear Color=<rgba 0.6879 0.3358 0.816 1>
     Scale Factor=&MakeFloat4FromVectorAndFloat("Vector (RGB)"=$System.ToD, "Float (Alpha)"=1)
    )
    Emitter.Color_02:LinearColor=&MultiplyLinearColor(
     Linear Color=<rgba 0.6879 0.3358 0.816 1>
     Scale Factor=&MakeFloat4FromVectorAndFloat("Vector (RGB)"=$System.ToD, "Float (Alpha)"=1)
    )
    Emitter.Color_03:LinearColor=&MultiplyLinearColor(
     Linear Color=&VariantColorSelector(Variant Amount=0, Original Color=<rgba 0.0018 0.0885 0.0857 1>, Variant 1 Color=<rgba 0.125 0.0182 0.1125 1>, Variant 2 Color=?)
     Scale Factor=&MakeFloat4FromVectorAndFloat("Vector (RGB)"=$System.ToD, "Float (Alpha)"=1)
    )
    Emitter.Color_04:LinearColor=&MultiplyLinearColor(
     Linear Color=<rgba 0.2812 0.041 0.2452 1>
     Scale Factor=&MakeFloat4FromVectorAndFloat("Vector (RGB)"=$System.ToD, "Float (Alpha)"=1)
    )
   ) {isSetParameters=true}
  ]
  particleSpawnScript [
   &PartitionParticles(Number Of Partitions=8)
   &InitializeParticle(
    Color=&MultiplyLinearColor(
     Linear Color=&VariantColorSelector(Variant Amount=0, Original Color=<rgba 0.0018 0.0885 0.0857 1>, Variant 1 Color=<rgba 0.125 0.0182 0.1125 1>, Variant 2 Color=?)
     Scale Factor=&MakeFloat4FromVectorAndFloat("Vector (RGB)"=$System.ToD, "Float (Alpha)"=1)
    )
    Lifetime Max=15
    Lifetime Min=12
    Lifetime Mode=::NewEnumerator1
    Mass Max=1.5
    Mass Mode=::NewEnumerator3
    Mesh Scale Mode=::NewEnumerator3
    Position Offset=&SelectVectorFromArray(
     Direct Array Index=$Particles.PartitionParticles.Partition
     Vector Selection Array=DataInterface() {propertyValues {FloatData [<-2 -2.5 0>, <0 0 0>, <0 0 0>, <0 0 0>, <0 0 -0.5>, <0 0 0>, <0 0 0>, <0 0 0>]}}
    )
    Sprite Rotation Mode=::NewEnumerator2
    UsePositionOffset=true
    Write Material Random=true
   )
   &CameraOffset(Camera Offset Amount=0.6) {enabled=false}
   /SetVariables(Particles.MeshIndex=$Particles.PartitionParticles.Partition) {isSetParameters=true}
  ]
  particleUpdateScript [
   &ParticleState(Kill Particles When Lifetime Has Elapsed=false)
   &ScaleMeshSize(Scale Factor=&MakeVector(X=1, Y=1, Z=&Waveform("Waveform [1]"=::NewEnumerator12, "[1] Frequency"=1, "[1] Amplitude Min/Max"=<1 1.5>))) {enabled=false}
   /SetVariables(Particles.AlphaScale:NiagaraFloat=&Multiply_Float(A=$System.Timeline001.Time, B=1)) {isSetParameters=true}
   &ScaleColor(Scale Alpha=$Particles.AlphaScale, Scale RGB=&VectorFromFloat(Value=10), ScaleA=true, ScaleRGB=true)
   /SetVariables(Particles.Linear Color=&Multiply_LinearColorByFloat(Linear Color=<rgba 0.1721 0.7031 0.4907 1>, Float=1)) {enabled=false, isSetParameters=true}
   /SetVariables(Particles.Scael02:Vector3f=&Multiply_VectorByFloat(Vector=$Particles.Scale, Float=1)) {enabled=false, isSetParameters=true}
   &DynamicMaterialParameters(
    Index 0 Param 1=0.5
    Index 0 Param 2=&RandomRangeFloat(Minimum=0.5, Maximum=1)
    Index 0 Param 3=&RandomRangeFloat(Minimum=0.5, Maximum=1)
    Index 0 Param 4=&RandomRangeFloat(Minimum=0.5, Maximum=1)
    Module.Index0Param0Write=true
    Module.Index0Param1Write=true
    Module.Index0Param2Write=true
    Module.Index0Param3Write=true
   ) {enabled=false}
   &FN_Light_Attributes(
    Color=&Multiply_LinearColorByFloat(Linear Color=$Emitter.Color_03, Float=$System.Timeline001.Time)
    Create Color Attribute=true
    Create Exponent Attribute=true
    Create Position Attribute=true
    Create Radius Attribute=true
    Enable Particle Light Enable=true
    Exponent=1
    Light Enabled=true
    Position=&AddVectorToPosition(Position=$Particles.Position, Vector=<0 -20 10>)
    Radius=15
    Scale Color=&Multiply_Float(
     A=/BRCosmetics/Weapons/FORT_Melee/Pickaxe_GateHound/FX/VariantManager/VariantFloatSelector(
      Original Float:NiagaraFloat=20
      Variant 1 Float:NiagaraFloat=7
      Variant 2 Float:NiagaraFloat=?
      Variant 3 Float:NiagaraFloat=?
      Variant Amount:NiagaraInt32=0
     )
     B=&RemapRange(
      Input Value To Remap=&FloatFromCurve(
       FloatCurve=DataInterface() {propertyValues {
         bOptimizeLUT=false
         bUseLUT=false
         Curve {
          postInfinityExtrap=RCCE_Cycle
          preInfinityExtrap=RCCE_Cycle
          keys [
           <at 0 1 RCIM_Cubic>
           {arriveTangent=0.5283, interpMode=RCIM_Cubic, leaveTangent=0.5283, time=0.0746, value=0.875}
           {arriveTangent=-0.0227, interpMode=RCIM_Cubic, leaveTangent=-0.0227, tangentMode=RCTM_User, time=0.1682, value=1.0888}
           {arriveTangent=-1.3938, interpMode=RCIM_Cubic, leaveTangent=-1.3938, time=0.2019, value=0.6577}
           {arriveTangent=1.4729, interpMode=RCIM_Cubic, leaveTangent=1.4729, time=0.2748, value=0.9402}
           {arriveTangent=0.905, interpMode=RCIM_Cubic, leaveTangent=0.905, time=0.3464, value=0.8705}
           {arriveTangent=-1.2159, interpMode=RCIM_Cubic, leaveTangent=-1.2159, time=0.4074, value=1.0603}
           {arriveTangent=-1.5422, interpMode=RCIM_Cubic, leaveTangent=-1.5422, time=0.4365, value=0.7609}
           {arriveTangent=-0.8529, interpMode=RCIM_Cubic, leaveTangent=-0.8529, time=0.4871, value=0.9373}
           {arriveTangent=0.3027, interpMode=RCIM_Cubic, leaveTangent=0.3027, time=0.5331, value=0.6785}
           {arriveTangent=1.4034, interpMode=RCIM_Cubic, leaveTangent=1.4034, time=0.6112, value=0.9748}
           {arriveTangent=1.5913, interpMode=RCIM_Cubic, leaveTangent=1.5913, time=0.69, value=0.8986}
           {arriveTangent=-0.751, interpMode=RCIM_Cubic, leaveTangent=-0.751, time=0.7536, value=1.2014}
           {arriveTangent=-2.595, interpMode=RCIM_Cubic, leaveTangent=-2.595, time=0.7995, value=0.8164}
           {arriveTangent=-0.8671, interpMode=RCIM_Cubic, leaveTangent=-0.8671, time=0.8674, value=0.9059}
           {arriveTangent=-0.3671, interpMode=RCIM_Cubic, leaveTangent=-0.3671, time=0.9138, value=0.7173}
           {arriveTangent=-0.6714, interpMode=RCIM_Cubic, leaveTangent=-0.6714, time=0.9899, value=0.8609}
           {arriveTangent=0.7195, interpMode=RCIM_Cubic, leaveTangent=0.7195, time=1.0372, value=0.6345}
           {arriveTangent=1.5649, interpMode=RCIM_Cubic, leaveTangent=1.5649, time=1.099, value=0.9394}
           {arriveTangent=0.3682, interpMode=RCIM_Cubic, leaveTangent=0.3682, time=1.1296, value=0.7792}
           {arriveTangent=1.199, interpMode=RCIM_Cubic, leaveTangent=1.199, time=1.2078, value=0.9795}
           {arriveTangent=0.6999, interpMode=RCIM_Cubic, leaveTangent=0.6999, time=1.2281, value=0.8972}
           {arriveTangent=0.8013, interpMode=RCIM_Cubic, leaveTangent=0.8013, time=1.329, value=1.0643}
           {arriveTangent=0.4649, interpMode=RCIM_Cubic, leaveTangent=0.4649, time=1.3719, value=1.0125}
           {arriveTangent=-3.0415, interpMode=RCIM_Cubic, leaveTangent=-3.0415, time=1.414, value=1.1038}
           {arriveTangent=-0.0928, interpMode=RCIM_Cubic, leaveTangent=-0.0928, tangentMode=RCTM_User, time=1.4501, value=0.7746}
           {arriveTangent=0.7682, interpMode=RCIM_Cubic, leaveTangent=0.7682, time=1.5242, value=0.9899}
           {arriveTangent=0.5284, interpMode=RCIM_Cubic, leaveTangent=0.5284, time=1.5566, value=0.8565}
           {arriveTangent=-0.1056, interpMode=RCIM_Cubic, leaveTangent=-0.1056, time=1.7244, value=1.0957}
           {arriveTangent=-2.1193, interpMode=RCIM_Cubic, leaveTangent=-2.1193, time=1.7619, value=0.8348}
           {arriveTangent=-1.2854, interpMode=RCIM_Cubic, leaveTangent=-1.2854, time=1.7959, value=0.9442}
           {arriveTangent=-0.1279, interpMode=RCIM_Cubic, leaveTangent=-0.1279, time=1.836, value=0.7395}
           {arriveTangent=1.0187, interpMode=RCIM_Cubic, leaveTangent=1.0187, time=1.9343, value=0.9265}
           {arriveTangent=1.0782, interpMode=RCIM_Cubic, leaveTangent=1.0782, time=1.9912, value=0.8977}
           {arriveTangent=-2.5847, interpMode=RCIM_Cubic, leaveTangent=-2.5847, time=2.0488, value=1.0499}
           {arriveTangent=-0.6379, interpMode=RCIM_Cubic, leaveTangent=-0.6379, time=2.0828, value=0.6611}
           {arriveTangent=0.4353, interpMode=RCIM_Cubic, leaveTangent=0.4353, time=2.2158, value=0.9434}
           {arriveTangent=-0.3447, interpMode=RCIM_Cubic, leaveTangent=-0.3447, time=2.2673, value=0.7414}
           {arriveTangent=0.6603, interpMode=RCIM_Cubic, leaveTangent=0.6603, time=2.3258, value=0.9055}
           {arriveTangent=0.3801, interpMode=RCIM_Cubic, leaveTangent=0.3801, time=2.3971, value=0.8271}
           {arriveTangent=0.6687, interpMode=RCIM_Cubic, leaveTangent=0.6687, time=2.4654, value=0.9586}
           {arriveTangent=1.2527, interpMode=RCIM_Cubic, leaveTangent=1.2527, time=2.5143, value=0.9055}
           {arriveTangent=-0.515, interpMode=RCIM_Cubic, leaveTangent=-0.515, time=2.5566, value=1.0728}
           {arriveTangent=-1.6201, interpMode=RCIM_Cubic, leaveTangent=-1.6201, time=2.597, value=0.8629}
           {arriveTangent=-0.9669, interpMode=RCIM_Cubic, leaveTangent=-0.9669, time=2.6456, value=0.9286}
           {arriveTangent=0.1009, interpMode=RCIM_Cubic, leaveTangent=0.1009, time=2.6921, value=0.771}
           {arriveTangent=0.0929, interpMode=RCIM_Cubic, leaveTangent=0.0929, time=2.7509, value=0.9392}
           {arriveTangent=0.2939, interpMode=RCIM_Cubic, leaveTangent=0.2939, time=2.8654, value=0.7871}
           {arriveTangent=1.7227, interpMode=RCIM_Cubic, leaveTangent=1.7227, time=2.9271, value=0.991}
           {arriveTangent=0.1236, interpMode=RCIM_Cubic, leaveTangent=0.1236, time=2.957, value=0.9449}
           <at 3 1 RCIM_Cubic>
          ]
         }
        }}
       CurveIndex=&Multiply_Float(A=$Particles.Age, B=0.4)
      )
      Clamp Results=false
      Input Min=0.6
      Input Max=1.2
      Output Min=0.6
      Output Max=1
     )
    )
   )
   &DynamicMaterialParameters(
    Index 0 Param 1=&RandomRangeFloat(Minimum=0, Maximum=1)
    Index 0 Param 2=&Multiply_Float(
     A=0.4
     B=&SelectFloatFromArray(
      Direct Array Index=&ReturnExecIndex()
      Float Selection Array=DataInterface() {propertyValues {FloatData [1,1,1,1,1,1,1,1]}}
     )
    )
    Module.Param0WriteEnabled=true
    Module.Param1WriteEnabled=true
    Module.Param2WriteEnabled=true
    Module.Param3WriteEnabled=true
   ) {module=DynamicMaterialParameters001}
   /SetVariables(Particles.Color_02:LinearColor=&MultiplyLinearColor(
     Linear Color=&MultiplyLinearColor(
      Linear Color=&VariantColorSelector(
       Variant Amount=0
       Original Color=&SelectLinearColorFromArray(
        Array Sampling Mode=::NewEnumerator1
        Direct Array Index=$Particles.PartitionParticles.Partition
        Color Selection Array=DataInterface() {propertyValues {ColorData [
           <rgba 0.0018 0.0885 0.0857 0.5>
           <rgba 0.0018 0.0885 0.0857 1>
           <rgba 0.0018 0.0885 0.0857 1>
           <rgba 0.0018 0.0885 0.0857 1>
           <rgba 0.0018 0.0885 0.0857 0.5>
           <rgba 0.0018 0.0885 0.0857 1>
           <rgba 0.0018 0.0885 0.0857 1>
           <rgba 0.0018 0.0885 0.0857 1>
          ]}}
        Random Seed=?
       )
       Variant 1 Color=<rgba 0.125 0.0182 0.1125 1>
       Variant 2 Color=?
      )
      Scale Factor=&MakeFloat4FromVectorAndFloat("Vector (RGB)"=$System.ToD, "Float (Alpha)"=$Particles.AlphaScale)
     )
     Scale Factor=&MakeFloat4FromVectorAndFloat("Vector (RGB)"=<1 1 1>, "Float (Alpha)"=1.5)
    )) {isSetParameters=true}
   /SetVariables(Particles.Color_03:LinearColor=&MultiplyLinearColor(
     Linear Color=&MultiplyLinearColor(
      Linear Color=&VariantColorSelector(
       Variant Amount=0
       Original Color=&SelectLinearColorFromArray(
        Array Sampling Mode=::NewEnumerator1
        Direct Array Index=$Particles.PartitionParticles.Partition
        Color Selection Array=DataInterface() {propertyValues {ColorData [
           <rgba 0.2203 0.9 0.7754 0.2>
           <rgba 0.2203 0.9 0.7754 1>
           <rgba 0.2203 0.9 0.7754 1>
           <rgba 0.2203 0.9 0.7754 1>
           <rgba 0.2203 0.9 0.7754 0.5>
           <rgba 0.2203 0.9 0.7754 1>
           <rgba 0.2203 0.9 0.7754 1>
           <rgba 0.2203 0.9 0.7754 1>
          ]}}
        Random Seed=?
       )
       Variant 1 Color=<rgba 0.6879 0.3358 0.816 1>
       Variant 2 Color=?
      )
      Scale Factor=&MakeFloat4FromVectorAndFloat("Vector (RGB)"=$System.ToD, "Float (Alpha)"=$Particles.AlphaScale)
     )
     Scale Factor=&MakeFloat4FromVectorAndFloat("Vector (RGB)"=<1 1 1>, "Float (Alpha)"=3)
    )) {isSetParameters=true}
  ]
  rendererClasses [&NiagaraLightRendererProperties, &NiagaraMeshRendererProperties]
  renderers [
   {rendererClass=&NiagaraLightRendererProperties, rendererIndex=0}
   {rendererClass=&NiagaraMeshRendererProperties, rendererIndex=1}
   {rendererClass=&NiagaraMeshRendererProperties, rendererIndex=2}
   {rendererClass=&NiagaraMeshRendererProperties, rendererIndex=3}
   {rendererClass=&NiagaraMeshRendererProperties, rendererIndex=4}
   {rendererClass=&NiagaraMeshRendererProperties, rendererIndex=5}
  ]
 }
 Flare001 {
  emitterUpdateScript [
   &EmitterState(Life Cycle Mode=::NewEnumerator1, Loop Behavior=::NewEnumerator1, Loop Duration Mode=::NewEnumerator1, UseLoopDelay=false)
   &SpawnBurst_Instantaneous(
    Spawn Count=&MakeCustomIntFromBool(Bool=&Boolean_OrOperation(Bool1=$User.bIsFrontEnd, Bool2=$User.bIsFrontEndPreview), True INT=0, False INT=1)
    Use Spawn Probability=false
    UseLoopCountLimit=false
   )
  ]
  particleSpawnScript [
   &InitializeParticle(
    Color=&MultiplyLinearColor(
     Linear Color=&VariantColorSelector(Variant Amount=0, Original Color=<rgba 0.1721 0.7031 0.6058 1>, Variant 1 Color=<rgba 0.5505 0.2014 0.6042 1>, Variant 2 Color=?)
     Scale Factor=<1 1 1 1>
    )
    Lifetime=0.3
    Sprite Size Mode=::NewEnumerator3
    Uniform Sprite Size=20
    UsePositionOffset=false
    Write Material Random=true
   )
   &SkeletalMeshLocation(MaskSpawnGroup=false, Mesh Sampling Type=::NewEnumerator5, Skeletal Mesh=$User.Skeletal Mesh, Socket Sampling Mode=::NewEnumerator5)
   &CameraOffset(Camera Offset Amount=3)
  ]
  particleUpdateScript [
   &ParticleState(Kill Particles When Lifetime Has Elapsed=false, Let Infinitely Lived Particles Die When Emitter Deactivates=true)
   &ScaleSpriteSize(
    Uniform Curve Index=$System.Timeline002.Time
    Uniform Curve Scale=1.5
    Uniform Curve Sprite Scale=DataInterface() {propertyValues {Curve {keys [
        {arriveTangent=2.4587, interpMode=RCIM_Cubic, leaveTangent=2.4587, tangentMode=RCTM_User, time=0, value=0}
        {arriveTangent=0.0649, interpMode=RCIM_Cubic, leaveTangent=0.0649, tangentMode=RCTM_User, time=1, value=1}
       ]}}}
   )
   &ScaleColor(
    Scale Alpha=&FloatFromCurve(
     FloatCurve=DataInterface() {propertyValues {Curve {keys [
         {arriveTangent=9.0725, interpMode=RCIM_Cubic, leaveTangent=9.0725, tangentMode=RCTM_User, time=0, value=0}
         {arriveTangent=-2.5249, interpMode=RCIM_Cubic, leaveTangent=-2.5249, tangentMode=RCTM_User, time=0.1408, value=0.9847}
         {arriveTangent=-0.0405, interpMode=RCIM_Cubic, leaveTangent=-0.0405, tangentMode=RCTM_User, time=1, value=0}
        ]}}}
     CurveIndex=$System.Timeline002.Time
     Scale Curve=10
    )
    Scale RGB=$System.ToD
    ScaleA=true
    ScaleRGB=true
   )
   &FN_Sprite_Attributes(
    Color=&MultiplyLinearColor(
     Linear Color=$Particles.Color
     Scale Factor=&MakeFloat4FromVectorAndFloat("Vector (RGB)"=<1 1 1>, "Float (Alpha)"=0.5)
    )
    Create Color Attribute=true
    Create SpriteSize Attribute=true
    SpriteSize=&Multiply_Vector2D_ByFloat(Float=2.5, Vector2D=$Particles.SpriteSize)
    Write to Module Namespace=true
   )
   &KillParticles(Kill Particles=&SetBoolByFloatComparison(A=$System.Timeline002.NormalizedTime, B=1, Comparison Type=::NewEnumerator4))
  ]
  rendererClasses [&NiagaraSpriteRendererProperties]
  renderers [
   {rendererClass=&NiagaraSpriteRendererProperties}
   {rendererClass=&NiagaraSpriteRendererProperties, rendererIndex=1}
  ]
 }
 Flare2001 {
  emitterUpdateScript [
   &EmitterState(Life Cycle Mode=::NewEnumerator1, Loop Behavior=::NewEnumerator1, Loop Duration Mode=::NewEnumerator1, UseLoopDelay=false)
   &SpawnBurst_Instantaneous(
    Spawn Count=&MakeCustomIntFromBool(Bool=&Boolean_OrOperation(Bool1=$User.bIsFrontEnd, Bool2=$User.bIsFrontEndPreview), True INT=0, False INT=1)
    Use Spawn Probability=false
    UseLoopCountLimit=false
   )
   &SpawnBurst_Instantaneous(
    Spawn Count=&MakeCustomIntFromBool(Bool=&Boolean_OrOperation(Bool1=$User.bIsFrontEnd, Bool2=$User.bIsFrontEndPreview), True INT=0, False INT=1)
    Spawn Group=1
    Use Spawn Probability=false
    UseLoopCountLimit=false
   ) {module=SpawnBurst_Instantaneous001}
  ]
  particleSpawnScript [
   &InitializeParticle(
    Color=&MultiplyLinearColor(
     Linear Color=&VariantColorSelector(Variant Amount=0, Original Color=<rgba 0.1721 0.7031 0.6058 1>, Variant 1 Color=<rgba 0.5505 0.2014 0.6042 1>, Variant 2 Color=?)
     Scale Factor=&MakeFloat4FromVectorAndFloat("Vector (RGB)"=$System.ToD, "Float (Alpha)"=1)
    )
    Lifetime=0.3
    Sprite Size Mode=::NewEnumerator3
    Uniform Sprite Size=35
    UsePositionOffset=false
    Write Material Random=true
   )
   /SetVariables(Particles.SpawnGroup=$Emitter.SpawnGroup) {isSetParameters=true, module=SetVariables_#3}
   &SkeletalMeshLocation(
    Bone / Socket Index=&Add_Integer(A=7, B=$Particles.SpawnGroup)
    MaskSpawnGroup=false
    Mesh Sampling Type=::NewEnumerator5
    Skeletal Mesh=$User.Skeletal Mesh
    Socket Sampling Mode=::NewEnumerator5
   )
   &CameraOffset(Camera Offset Amount=10)
   /SetVariables(Particles.VisibilityTag=$Particles.SpawnGroup) {isSetParameters=true, module=SetVariables_#4}
  ]
  particleUpdateScript [
   &ParticleState(Kill Particles When Lifetime Has Elapsed=false, Let Infinitely Lived Particles Die When Emitter Deactivates=true)
   &ScaleSpriteSize(
    Uniform Curve Index=$System.Timeline003.Time
    Uniform Curve Scale=1.5
    Uniform Curve Sprite Scale=DataInterface() {propertyValues {Curve {keys [
        {arriveTangent=2.4587, interpMode=RCIM_Cubic, leaveTangent=2.4587, tangentMode=RCTM_User, time=0, value=0}
        {arriveTangent=0.0649, interpMode=RCIM_Cubic, leaveTangent=0.0649, tangentMode=RCTM_User, time=1, value=1}
       ]}}}
   )
   &ScaleColor(
    Scale Alpha=&FloatFromCurve(
     FloatCurve=DataInterface() {propertyValues {Curve {keys [
         {arriveTangent=0.1248, interpMode=RCIM_Cubic, leaveTangent=0.1248, tangentMode=RCTM_User, time=0, value=0}
         {arriveTangent=-2.5249, interpMode=RCIM_Cubic, leaveTangent=-2.5249, tangentMode=RCTM_User, time=0.1408, value=0.9847}
         {arriveTangent=-0.0405, interpMode=RCIM_Cubic, leaveTangent=-0.0405, tangentMode=RCTM_User, time=1, value=0}
        ]}}}
     CurveIndex=$System.Timeline003.Time
     Scale Curve=3
    )
    Scale RGB=$System.ToD
    ScaleA=true
    ScaleRGB=true
   )
   &FN_Sprite_Attributes(
    Color=&MultiplyLinearColor(
     Linear Color=$Particles.Color
     Scale Factor=&MakeFloat4FromVectorAndFloat("Vector (RGB)"=<1 1 1>, "Float (Alpha)"=0.5)
    )
    Create Color Attribute=true
    Create SpriteSize Attribute=true
    SpriteSize=&Multiply_Vector2D_ByFloat(Float=2.5, Vector2D=$Particles.SpriteSize)
    Write to Module Namespace=true
   )
   &KillParticles(Kill Particles=&SetBoolByFloatComparison(A=$System.Timeline003.Time, B=1, Comparison Type=::NewEnumerator4))
   /SetVariables(
    Particles.DecalCameraOffset=15
    Particles.DecalDyn:Vector4f=&MakeVector4(
     X=&FloatFromCurve(FloatCurve=DataInterface() {propertyValues {Curve {keys [<at 0.35 0.5>,<at 1 1>]}}}, CurveIndex=$System.Timeline003.Time)
     Y=0.8
     Z=&RandomRangeFloat(Minimum=0, Maximum=1)
     W=&RandomRangeFloat(Minimum=0, Maximum=0)
    )
    Particles.GlowDecalColor:LinearColor=&MultiplyLinearColor(
     Linear Color=&MultiplyLinearColor(
      Linear Color=&Multiply_LinearColorByFloat(
       Linear Color=&VariantColorSelector(Variant Amount=0, Original Color=<rgba 0.0054 0.195 0.184 1>, Variant 1 Color=<rgba 0.2812 0.041 0.2452 1>, Variant 2 Color=?)
       Float=&FloatFromCurve(
        FloatCurve=DataInterface() {propertyValues {Curve {keys [
            {arriveTangent=0.1248, interpMode=RCIM_Cubic, leaveTangent=0.1248, tangentMode=RCTM_User, time=0, value=0}
            {arriveTangent=-2.5249, interpMode=RCIM_Cubic, leaveTangent=-2.5249, tangentMode=RCTM_User, time=0.35, value=0.9847}
            {arriveTangent=-0.0405, interpMode=RCIM_Cubic, leaveTangent=-0.0405, tangentMode=RCTM_User, time=1, value=0}
           ]}}}
        CurveIndex=$System.Timeline003.Time
        Scale Curve=3.5
       )
      )
      Scale Factor=&MakeFloat4FromVectorAndFloat(
       "Vector (RGB)"=&VectorFromFloat(Value=&FloatFromCurve(
         FloatCurve=DataInterface() {propertyValues {Curve {keys [
             {arriveTangent=0.1248, interpMode=RCIM_Cubic, leaveTangent=0.1248, tangentMode=RCTM_User, time=0, value=0}
             {arriveTangent=-2.5249, interpMode=RCIM_Cubic, leaveTangent=-2.5249, tangentMode=RCTM_User, time=0.35, value=0.9847}
             {interpMode=RCIM_Cubic, leaveTangent=-0.0405, time=1, value=0}
            ]}}}
         CurveIndex=$System.Timeline003.Time
         Scale Curve=10
        ))
       "Float (Alpha)"=&FloatFromCurve(
        FloatCurve=DataInterface() {propertyValues {Curve {keys [
            {arriveTangent=0.1248, interpMode=RCIM_Cubic, leaveTangent=0.1248, tangentMode=RCTM_User, time=0, value=0}
            {arriveTangent=-2.5249, interpMode=RCIM_Cubic, leaveTangent=-2.5249, tangentMode=RCTM_User, time=0.35, value=0.9847}
            {interpMode=RCIM_Cubic, leaveTangent=-0.0405, time=1, value=0}
           ]}}}
        CurveIndex=$System.Timeline003.Time
        Scale Curve=10
       )
      )
     )
     Scale Factor=&MakeFloat4FromVectorAndFloat("Vector (RGB)"=$System.DecalToD, "Float (Alpha)"=1)
    )
    Particles.GlowDecalSize:Vector2f=&Multiply_Vector2D_ByFloat(Float=28, Vector2D=<1 1>)
   ) {isSetParameters=true, module=SetVariables_#5}
   /SetVariables(Particles.HaloColor:LinearColor=&MultiplyLinearColor(
     Linear Color=$Particles.Color
     Scale Factor=&MakeFloat4FromVectorAndFloat("Vector (RGB)"=<1 1 1>, "Float (Alpha)"=0.5)
    )) {isSetParameters=true, module=SetVariables_#6}
  ]
  rendererClasses [&NiagaraSpriteRendererProperties]
  renderers [
   {rendererClass=&NiagaraSpriteRendererProperties}
   {rendererClass=&NiagaraSpriteRendererProperties, rendererIndex=1}
   {rendererClass=&NiagaraSpriteRendererProperties, rendererIndex=2}
  ]
 }
 Flare2FE {
  emitterUpdateScript [
   &EmitterState(
    Inactive Response=::NewEnumerator1
    Life Cycle Mode=::NewEnumerator1
    Loop Behavior=::NewEnumerator1
    Loop Duration Mode=::NewEnumerator1
    UseLoopDelay=false
   )
   &SpawnBurst_Instantaneous(
    Spawn Count=&MakeCustomIntFromBool(Bool=&Boolean_OrOperation(Bool1=$User.bIsFrontEnd, Bool2=$User.bIsFrontEndPreview), True INT=1, False INT=0)
    Use Spawn Probability=false
    UseLoopCountLimit=false
   )
   &SpawnBurst_Instantaneous(
    Spawn Count=&MakeCustomIntFromBool(Bool=&Boolean_OrOperation(Bool1=$User.bIsFrontEnd, Bool2=$User.bIsFrontEndPreview), True INT=1, False INT=0)
    Spawn Group=1
    Use Spawn Probability=false
    UseLoopCountLimit=false
   ) {module=SpawnBurst_Instantaneous001}
  ]
  particleSpawnScript [
   &InitializeParticle(
    Color=&MultiplyLinearColor(
     Linear Color=&VariantColorSelector(Variant Amount=0, Original Color=<rgba 0.1721 0.7031 0.6058 1>, Variant 1 Color=<rgba 0.5505 0.2014 0.6042 1>, Variant 2 Color=?)
     Scale Factor=&MakeFloat4FromVectorAndFloat("Vector (RGB)"=$System.ToD, "Float (Alpha)"=1)
    )
    Lifetime=0.3
    Sprite Size Mode=::NewEnumerator3
    Uniform Sprite Size=35
    UsePositionOffset=false
    Write Material Random=true
   )
   /SetVariables(Particles.SpawnGroup=$Emitter.SpawnGroup) {isSetParameters=true, module=SetVariables_#3}
   &SkeletalMeshLocation(
    Bone / Socket Index=&Add_Integer(A=7, B=$Particles.SpawnGroup)
    MaskSpawnGroup=false
    Mesh Sampling Type=::NewEnumerator5
    Skeletal Mesh=$User.Skeletal Mesh
    Socket Sampling Mode=::NewEnumerator5
   )
   &CameraOffset(Camera Offset Amount=10)
   /SetVariables(Particles.VisibilityTag=$Particles.SpawnGroup) {isSetParameters=true, module=SetVariables_#4}
  ]
  particleUpdateScript [
   &ParticleState(Kill Particles When Lifetime Has Elapsed=false, Let Infinitely Lived Particles Die When Emitter Deactivates=true)
   &ScaleSpriteSize(
    Uniform Curve Index=$System.Timeline003.Time
    Uniform Curve Scale=1.5
    Uniform Curve Sprite Scale=DataInterface() {propertyValues {Curve {keys [
        {arriveTangent=2.4587, interpMode=RCIM_Cubic, leaveTangent=2.4587, tangentMode=RCTM_User, time=0, value=0}
        {arriveTangent=0.0649, interpMode=RCIM_Cubic, leaveTangent=0.0649, tangentMode=RCTM_User, time=1, value=1}
       ]}}}
   )
   &ScaleColor(
    Scale Alpha=&FloatFromCurve(
     FloatCurve=DataInterface() {propertyValues {Curve {keys [
         {arriveTangent=0.1248, interpMode=RCIM_Cubic, leaveTangent=0.1248, tangentMode=RCTM_User, time=0, value=0}
         {arriveTangent=-2.5249, interpMode=RCIM_Cubic, leaveTangent=-2.5249, tangentMode=RCTM_User, time=0.1408, value=0.9847}
         {arriveTangent=-0.0405, interpMode=RCIM_Cubic, leaveTangent=-0.0405, tangentMode=RCTM_User, time=1, value=0}
        ]}}}
     CurveIndex=$System.Timeline003.Time
     Scale Curve=3
    )
    Scale RGB=$System.ToD
    ScaleA=true
    ScaleRGB=true
   )
   &FN_Sprite_Attributes(
    Color=&MultiplyLinearColor(
     Linear Color=$Particles.Color
     Scale Factor=&MakeFloat4FromVectorAndFloat("Vector (RGB)"=<1 1 1>, "Float (Alpha)"=0.5)
    )
    Create Color Attribute=true
    Create SpriteSize Attribute=true
    SpriteSize=&Multiply_Vector2D_ByFloat(Float=2.5, Vector2D=$Particles.SpriteSize)
    Write to Module Namespace=true
   )
   &KillParticles(Kill Particles=&SetBoolByFloatComparison(A=$System.Timeline003.Time, B=1, Comparison Type=::NewEnumerator4)) {enabled=false}
   /SetVariables(
    Particles.DecalCameraOffset=15
    Particles.DecalDyn:Vector4f=&MakeVector4(
     X=&FloatFromCurve(FloatCurve=DataInterface() {propertyValues {Curve {keys [<at 0.35 0.5>,<at 1 1>]}}}, CurveIndex=$System.Timeline003.Time)
     Y=0.8
     Z=&RandomRangeFloat(Minimum=0, Maximum=1)
     W=&RandomRangeFloat(Minimum=0, Maximum=0)
    )
    Particles.GlowDecalColor:LinearColor=&MultiplyLinearColor(
     Linear Color=&MultiplyLinearColor(
      Linear Color=&Multiply_LinearColorByFloat(
       Linear Color=&VariantColorSelector(Variant Amount=0, Original Color=<rgba 0.0054 0.195 0.184 1>, Variant 1 Color=<rgba 0.2812 0.041 0.2452 1>, Variant 2 Color=?)
       Float=&FloatFromCurve(
        FloatCurve=DataInterface() {propertyValues {Curve {keys [
            {arriveTangent=0.1248, interpMode=RCIM_Cubic, leaveTangent=0.1248, tangentMode=RCTM_User, time=0, value=0}
            {arriveTangent=-2.5249, interpMode=RCIM_Cubic, leaveTangent=-2.5249, tangentMode=RCTM_User, time=0.35, value=0.9847}
            {arriveTangent=-0.0405, interpMode=RCIM_Cubic, leaveTangent=-0.0405, tangentMode=RCTM_User, time=1, value=0}
           ]}}}
        CurveIndex=$System.Timeline003.Time
        Scale Curve=3.5
       )
      )
      Scale Factor=&MakeFloat4FromVectorAndFloat(
       "Vector (RGB)"=&VectorFromFloat(Value=&FloatFromCurve(
         FloatCurve=DataInterface() {propertyValues {Curve {keys [
             {arriveTangent=0.1248, interpMode=RCIM_Cubic, leaveTangent=0.1248, tangentMode=RCTM_User, time=0, value=0}
             {arriveTangent=-2.5249, interpMode=RCIM_Cubic, leaveTangent=-2.5249, tangentMode=RCTM_User, time=0.35, value=0.9847}
             {interpMode=RCIM_Cubic, leaveTangent=-0.0405, time=1, value=0}
            ]}}}
         CurveIndex=$System.Timeline003.Time
         Scale Curve=10
        ))
       "Float (Alpha)"=&FloatFromCurve(
        FloatCurve=DataInterface() {propertyValues {Curve {keys [
            {arriveTangent=0.1248, interpMode=RCIM_Cubic, leaveTangent=0.1248, tangentMode=RCTM_User, time=0, value=0}
            {arriveTangent=-2.5249, interpMode=RCIM_Cubic, leaveTangent=-2.5249, tangentMode=RCTM_User, time=0.35, value=0.9847}
            {interpMode=RCIM_Cubic, leaveTangent=-0.0405, time=1, value=0}
           ]}}}
        CurveIndex=$System.Timeline003.Time
        Scale Curve=10
       )
      )
     )
     Scale Factor=&MakeFloat4FromVectorAndFloat("Vector (RGB)"=$System.DecalToD, "Float (Alpha)"=1)
    )
    Particles.GlowDecalSize:Vector2f=&Multiply_Vector2D_ByFloat(Float=28, Vector2D=<1 1>)
   ) {isSetParameters=true, module=SetVariables_#5}
   /SetVariables(Particles.HaloColor:LinearColor=&MultiplyLinearColor(
     Linear Color=$Particles.Color
     Scale Factor=&MakeFloat4FromVectorAndFloat("Vector (RGB)"=<1 1 1>, "Float (Alpha)"=0.5)
    )) {isSetParameters=true, module=SetVariables_#6}
  ]
  rendererClasses [&NiagaraSpriteRendererProperties]
  renderers [
   {rendererClass=&NiagaraSpriteRendererProperties}
   {rendererClass=&NiagaraSpriteRendererProperties, rendererIndex=1}
   {rendererClass=&NiagaraSpriteRendererProperties, rendererIndex=2}
  ]
 }
 FlareFE {
  emitterUpdateScript [
   &EmitterState(
    Inactive Response=::NewEnumerator1
    Life Cycle Mode=::NewEnumerator1
    Loop Behavior=::NewEnumerator1
    Loop Duration Mode=::NewEnumerator1
    UseLoopDelay=false
   )
   &SpawnBurst_Instantaneous(
    Spawn Count=&MakeCustomIntFromBool(Bool=&Boolean_OrOperation(Bool1=$User.bIsFrontEnd, Bool2=$User.bIsFrontEndPreview), True INT=1, False INT=0)
    Use Spawn Probability=false
    UseLoopCountLimit=false
   )
  ]
  particleSpawnScript [
   &InitializeParticle(
    Color=&MultiplyLinearColor(
     Linear Color=&VariantColorSelector(Variant Amount=0, Original Color=<rgba 0.1721 0.7031 0.6058 1>, Variant 1 Color=<rgba 0.5505 0.2014 0.6042 1>, Variant 2 Color=?)
     Scale Factor=<1 1 1 1>
    )
    Lifetime=0.3
    Sprite Size Mode=::NewEnumerator3
    Uniform Sprite Size=20
    UsePositionOffset=false
    Write Material Random=true
   )
   &SkeletalMeshLocation(MaskSpawnGroup=false, Mesh Sampling Type=::NewEnumerator5, Skeletal Mesh=$User.Skeletal Mesh, Socket Sampling Mode=::NewEnumerator5)
   &CameraOffset(Camera Offset Amount=3)
  ]
  particleUpdateScript [
   &ParticleState(Kill Particles When Lifetime Has Elapsed=false, Let Infinitely Lived Particles Die When Emitter Deactivates=true)
   &ScaleSpriteSize(
    Uniform Curve Index=$System.Timeline002.Time
    Uniform Curve Scale=1.5
    Uniform Curve Sprite Scale=DataInterface() {propertyValues {Curve {keys [
        {arriveTangent=2.4587, interpMode=RCIM_Cubic, leaveTangent=2.4587, tangentMode=RCTM_User, time=0, value=0}
        {arriveTangent=0.0649, interpMode=RCIM_Cubic, leaveTangent=0.0649, tangentMode=RCTM_User, time=1, value=1}
       ]}}}
   )
   &ScaleColor(
    Scale Alpha=&FloatFromCurve(
     FloatCurve=DataInterface() {propertyValues {Curve {keys [
         {arriveTangent=9.0725, interpMode=RCIM_Cubic, leaveTangent=9.0725, tangentMode=RCTM_User, time=0, value=0}
         {arriveTangent=-2.5249, interpMode=RCIM_Cubic, leaveTangent=-2.5249, tangentMode=RCTM_User, time=0.1408, value=0.9847}
         {arriveTangent=-0.0405, interpMode=RCIM_Cubic, leaveTangent=-0.0405, tangentMode=RCTM_User, time=1, value=0}
        ]}}}
     CurveIndex=$System.Timeline002.Time
     Scale Curve=10
    )
    Scale RGB=$System.ToD
    ScaleA=true
    ScaleRGB=true
   )
   &FN_Sprite_Attributes(
    Color=&MultiplyLinearColor(
     Linear Color=$Particles.Color
     Scale Factor=&MakeFloat4FromVectorAndFloat("Vector (RGB)"=<1 1 1>, "Float (Alpha)"=0.5)
    )
    Create Color Attribute=true
    Create SpriteSize Attribute=true
    SpriteSize=&Multiply_Vector2D_ByFloat(Float=2.5, Vector2D=$Particles.SpriteSize)
    Write to Module Namespace=true
   )
   &KillParticles(Kill Particles=&SetBoolByFloatComparison(A=$System.Timeline002.NormalizedTime, B=1, Comparison Type=::NewEnumerator4)) {enabled=false}
  ]
  rendererClasses [&NiagaraSpriteRendererProperties]
  renderers [
   {rendererClass=&NiagaraSpriteRendererProperties}
   {rendererClass=&NiagaraSpriteRendererProperties, rendererIndex=1}
  ]
 }
 SimpleSpriteBurst001 {
  emitterUpdateScript [
   &EmitterState(
    Inactive Response=::NewEnumerator1
    Life Cycle Mode=::NewEnumerator1
    Loop Behavior=::NewEnumerator1
    Loop Duration Mode=::NewEnumerator1
    UseLoopDelay=false
   )
   &SpawnBurst_Instantaneous(Spawn Count=2, Use Spawn Probability=false, UseLoopCountLimit=false)
  ]
  particleSpawnScript [
   &PartitionParticles()
   &InitializeParticle(
    Color=&MultiplyLinearColor(
     Linear Color=&VariantColorSelector(Variant Amount=0, Original Color=<rgba 0.0018 0.0885 0.0857 1>, Variant 1 Color=<rgba 0.125 0.0182 0.1125 1>, Variant 2 Color=?)
     Scale Factor=&MakeFloat4FromVectorAndFloat("Vector (RGB)"=$System.ToD, "Float (Alpha)"=1)
    )
    Lifetime=2
    Mesh Scale Mode=::NewEnumerator3
    Sprite Size Mode=::NewEnumerator3
    Uniform Sprite Size=3.9
    UsePositionOffset=false
    Write Material Random=true
   )
   &CameraOffset(Camera Offset Amount=1)
   &InitialMeshOrientation(
    Mesh Orientation Mode=::NewEnumerator2
    Orientation Vector=<0 -1 0>
    Rotation=&SelectVectorFromArray(
     Direct Array Index=$Particles.PartitionParticles.Partition
     Vector Selection Array=DataInterface() {propertyValues {FloatData [<-0.05 -0.1 0.1>, <0.05 -0.1 -0.1>]}}
    )
    Use Rotation Vector=true
   ) {enabled=false}
   /SetVariables(Particles.MeshIndex=$Particles.PartitionParticles.Partition) {isSetParameters=true}
  ]
  particleUpdateScript [
   &ParticleState(Kill Particles When Lifetime Has Elapsed=false)
   &SkeletalMeshLocation(
    Bone / Socket Index=$Particles.PartitionParticles.Partition
    MaskSpawnGroup=false
    Mesh Sampling Type=::NewEnumerator5
    Skeletal Mesh=$User.Skeletal Mesh
    Socket Sampling Mode=::NewEnumerator5
   ) {enabled=false}
   /SetVariables(Particles.AlphaScale:NiagaraFloat=&Multiply_Float(A=2, B=$System.Timeline001.Time)) {isSetParameters=true, module=SetVariables_#7}
   /SetVariables(Particles.SplineThicknessMult=1) {isSetParameters=true, module=SetVariables_#8}
   /SetVariables(Particles.URand:NiagaraFloat=&RandomRangeFloat(Minimum=0, Maximum=1)) {isSetParameters=true, module=SetVariables_#9}
   &DynamicMaterialParameters(
    Index 0 Param 1=&Multiply_Float(A=1, B=$Particles.SplineThicknessMult)
    Index 0 Param 2=&Multiply_Float(A=5, B=$Particles.SplineThicknessMult)
    Index 0 Param 3=$Particles.URand
    Module.Index0Param0Write=true
    Module.Index0Param1Write=true
    Module.Index0Param2Write=true
    Module.Index0Param3Write=true
   )
   &ScaleColor(Scale Alpha=$Particles.AlphaScale, ScaleA=true, ScaleRGB=true)
   /SetVariables(
    Particles.Color2=&MultiplyLinearColor(
     Linear Color=&VariantColorSelector(Variant Amount=0, Original Color=<rgba 0.0048 0.6198 0.3447 0.35>, Variant 1 Color=<rgba 0.5505 0.2014 0.6042 0.35>, Variant 2 Color=?)
     Scale Factor=&MakeFloat4FromVectorAndFloat("Vector (RGB)"=$System.ToD, "Float (Alpha)"=$Particles.AlphaScale)
    )
    Particles.Dyn2:Vector4f=&MakeVector4(
     X=&Multiply_Float(A=1, B=$Particles.SplineThicknessMult)
     Y=&Multiply_Float(A=3.5, B=$Particles.SplineThicknessMult)
     Z=$Particles.URand
     W=0
    )
   ) {isSetParameters=true, module=SetVariables_#10}
   /SetVariables(
    Particles.Color3:LinearColor=&MultiplyLinearColor(
     Linear Color=&VariantColorSelector(Variant Amount=0, Original Color=<rgba 0.1721 0.7031 0.6058 0.5>, Variant 1 Color=<rgba 0.6879 0.3358 0.816 0.5>, Variant 2 Color=?)
     Scale Factor=&MakeFloat4FromVectorAndFloat("Vector (RGB)"=$System.ToD, "Float (Alpha)"=$Particles.AlphaScale)
    )
    Particles.Dyn3:Vector4f=&MakeVector4(
     X=&Multiply_Float(A=0.5, B=$Particles.SplineThicknessMult)
     Y=&Multiply_Float(A=3, B=$Particles.SplineThicknessMult)
     Z=$Particles.URand
     W=0
    )
   ) {isSetParameters=true, module=SetVariables_#11}
   /SetVariables(
    Particles.Color4:LinearColor=&MultiplyLinearColor(
     Linear Color=&VariantColorSelector(Variant Amount=0, Original Color=<rgba 0.0048 0.344 0.281 0.15>, Variant 1 Color=<rgba 0.2812 0.041 0.2452 0.15>, Variant 2 Color=?)
     Scale Factor=&MakeFloat4FromVectorAndFloat("Vector (RGB)"=$System.ToD, "Float (Alpha)"=$Particles.AlphaScale)
    )
    Particles.Dyn4:Vector4f=&MakeVector4(
     X=&Multiply_Float(A=1, B=$Particles.SplineThicknessMult)
     Y=&Multiply_Float(A=4, B=$Particles.SplineThicknessMult)
     Z=$Particles.URand
     W=0
    )
   ) {isSetParameters=true, module=SetVariables_#12}
   /SetVariables(Particles.AirGlowColor:LinearColor=&Multiply_LinearColorByFloat(
     Linear Color=&VariantColorSelector(Variant Amount=0, Original Color=<rgba 0.0776 0.5729 0.4053 1>, Variant 1 Color=<rgba 1 1 1 1>, Variant 2 Color=?)
     Float=1
    )) {enabled=false, isSetParameters=true, module=SetVariables_#13}
   /SetVariables(
    Particles.DecalDyn:Vector4f=&MakeVector4(
     X=0.5
     Y=0.8
     Z=&RandomRangeFloat(Minimum=0, Maximum=1)
     W=&RandomRangeFloat(Minimum=0, Maximum=0)
    )
    Particles.FireDecalColor:LinearColor=&Multiply_LinearColorByFloat(
     Linear Color=&VariantColorSelector(Variant Amount=0, Original Color=<rgba 0.1721 0.7031 0.6058 1>, Variant 1 Color=<rgba 1 1 1 1>, Variant 2 Color=?)
     Float=3.5
    )
    Particles.FireDecalSize:Vector2f=&Multiply_Vector2D_ByFloat(Float=0.65, Vector2D=$Particles.SpriteSize)
    Particles.GlowDecalColor:LinearColor=&Multiply_LinearColorByFloat(
     Linear Color=&VariantColorSelector(Variant Amount=0, Original Color=<rgba 0.0054 0.195 0.184 1>, Variant 1 Color=<rgba 1 1 1 1>, Variant 2 Color=?)
     Float=3.5
    )
   ) {enabled=false, isSetParameters=true, module=SetVariables_#14}
  ]
  rendererClasses [&NiagaraSpriteRendererProperties, &NiagaraMeshRendererProperties]
  renderers [
   {rendererClass=&NiagaraSpriteRendererProperties}
   {rendererClass=&NiagaraMeshRendererProperties, rendererIndex=1}
   {rendererClass=&NiagaraMeshRendererProperties, rendererIndex=2}
   {rendererClass=&NiagaraMeshRendererProperties, rendererIndex=3}
   {rendererClass=&NiagaraMeshRendererProperties, rendererIndex=4}
   {rendererClass=&NiagaraSpriteRendererProperties, rendererIndex=5}
   {rendererClass=&NiagaraSpriteRendererProperties, rendererIndex=6}
  ]
 }
 SimpleSpriteBurst003 {
  emitterUpdateScript [
   &EmitterState(
    Inactive Response=::NewEnumerator1
    Life Cycle Mode=::NewEnumerator1
    Loop Behavior=::NewEnumerator1
    Loop Duration Mode=::NewEnumerator1
    UseLoopDelay=false
   )
   &SpawnBurst_Instantaneous(Spawn Count=2, Use Spawn Probability=false, UseLoopCountLimit=false)
  ]
  particleSpawnScript [
   &InitializeParticle(
    Color=&MultiplyLinearColor(
     Linear Color=&VariantColorSelector(Variant Amount=0, Original Color=<rgba 0.0018 0.0885 0.0857 1>, Variant 1 Color=<rgba 0.125 0.0182 0.1125 1>, Variant 2 Color=?)
     Scale Factor=&MakeFloat4FromVectorAndFloat("Vector (RGB)"=$System.ToD, "Float (Alpha)"=1)
    )
    Lifetime=2
    Mesh Scale Mode=::NewEnumerator3
    Mesh Uniform Scale=0.15
    Sprite Size Mode=::NewEnumerator3
    Uniform Sprite Size=&VariantFloatSelector(
     Original Float:NiagaraFloat=3.9
     Variant 1 Float:NiagaraFloat=2
     Variant 2 Float:NiagaraFloat=?
     Variant 3 Float:NiagaraFloat=?
     Variant Amount:NiagaraInt32=0
    )
    UsePositionOffset=false
    Write Material Random=true
   )
   &PartitionParticles()
   &CameraOffset(Camera Offset Amount=2)
   &InitialMeshOrientation(
    Mesh Orientation Mode=::NewEnumerator2
    Orientation Vector=<0 -1 0>
    Rotation=&SelectVectorFromArray(
     Direct Array Index=$Particles.PartitionParticles.Partition
     Vector Selection Array=DataInterface() {propertyValues {FloatData [<-0.05 -0.3 0.1>, <0.05 -0.3 -0.1>]}}
    )
    Use Rotation Vector=true
   )
  ]
  particleUpdateScript [
   &ParticleState(Kill Particles When Lifetime Has Elapsed=false, Loop Particles Lifetime=true)
   &SkeletalMeshLocation(
    Bone / Socket Index=$Particles.PartitionParticles.Partition
    MaskSpawnGroup=false
    Mesh Sampling Type=::NewEnumerator5
    Skeletal Mesh=$User.Skeletal Mesh
    Socket Sampling Mode=::NewEnumerator5
   )
   /SetVariables(Particles.AlphaScale:NiagaraFloat=&Lerp_Float(
     A=&Lerp_Float(A=0, B=2, Alpha=$System.Timeline.Time)
     B=&Lerp_Float(A=0, B=2, Alpha=$System.Timeline001.Time)
     Alpha=&MakeFloatFromInt(Int=$Particles.PartitionParticles.Partition)
    )) {isSetParameters=true, module=SetVariables_#7}
   /SetVariables(Particles.SplineThicknessMult:NiagaraFloat=&VariantFloatSelector(
     Original Float:NiagaraFloat=1.2
     Variant 1 Float:NiagaraFloat=0.8
     Variant 2 Float:NiagaraFloat=?
     Variant 3 Float:NiagaraFloat=?
     Variant Amount:NiagaraInt32=0
    )) {isSetParameters=true, module=SetVariables_#8}
   /SetVariables(Particles.URand:NiagaraFloat=&SelectFloatFromArray(Direct Array Index=$Particles.PartitionParticles.Partition, Float Selection Array=DataInterface() {propertyValues {FloatData [0,0.5]}})) {isSetParameters=true, module=SetVariables_#9}
   &DynamicMaterialParameters(
    Index 0 Param 1=&Multiply_Float(A=1, B=$Particles.SplineThicknessMult)
    Index 0 Param 2=&Multiply_Float(A=5, B=$Particles.SplineThicknessMult)
    Index 0 Param 3=$Particles.URand
    Module.Index0Param0Write=true
    Module.Index0Param1Write=true
    Module.Index0Param2Write=true
    Module.Index0Param3Write=true
   )
   &ScaleColor(Scale Alpha=$Particles.AlphaScale, ScaleA=true, ScaleRGB=true)
   /SetVariables(
    Particles.Color2=&MultiplyLinearColor(
     Linear Color=&VariantColorSelector(Variant Amount=0, Original Color=<rgba 0.1721 0.7031 0.6058 1>, Variant 1 Color=<rgba 0.5505 0.2014 0.6042 1>, Variant 2 Color=?)
     Scale Factor=&MakeFloat4FromVectorAndFloat("Vector (RGB)"=$System.ToD, "Float (Alpha)"=$Particles.AlphaScale)
    )
    Particles.Dyn2:Vector4f=&MakeVector4(
     X=&Multiply_Float(A=1, B=$Particles.SplineThicknessMult)
     Y=&Multiply_Float(A=3.5, B=$Particles.SplineThicknessMult)
     Z=$Particles.URand
     W=0
    )
   ) {isSetParameters=true, module=SetVariables_#10}
   /SetVariables(
    Particles.Color3:LinearColor=&MultiplyLinearColor(
     Linear Color=&VariantColorSelector(Variant Amount=0, Original Color=<rgba 0.1721 0.7031 0.6058 0.5>, Variant 1 Color=<rgba 0.6879 0.3358 0.816 2>, Variant 2 Color=?)
     Scale Factor=&MakeFloat4FromVectorAndFloat("Vector (RGB)"=$System.ToD, "Float (Alpha)"=$Particles.AlphaScale)
    )
    Particles.Dyn3:Vector4f=&MakeVector4(
     X=&Multiply_Float(A=0.5, B=$Particles.SplineThicknessMult)
     Y=&Multiply_Float(A=3, B=$Particles.SplineThicknessMult)
     Z=$Particles.URand
     W=0
    )
   ) {isSetParameters=true, module=SetVariables_#11}
   /SetVariables(
    Particles.Color4:LinearColor=&MultiplyLinearColor(
     Linear Color=&VariantColorSelector(Variant Amount=0, Original Color=<rgba 0.0048 0.344 0.281 0.15>, Variant 1 Color=<rgba 0.2812 0.041 0.2452 0.15>, Variant 2 Color=?)
     Scale Factor=&MakeFloat4FromVectorAndFloat("Vector (RGB)"=$System.ToD, "Float (Alpha)"=$Particles.AlphaScale)
    )
    Particles.Dyn4:Vector4f=&MakeVector4(
     X=&Multiply_Float(A=1, B=$Particles.SplineThicknessMult)
     Y=&Multiply_Float(A=4, B=$Particles.SplineThicknessMult)
     Z=$Particles.URand
     W=0
    )
   ) {isSetParameters=true, module=SetVariables_#12}
   /SetVariables(Particles.AirGlowColor:LinearColor=&MultiplyLinearColor(
     Linear Color=&Multiply_LinearColorByFloat(
      Linear Color=&VariantColorSelector(Variant Amount=0, Original Color=<rgba 0.0776 0.5729 0.4053 1>, Variant 1 Color=<rgba 0.2812 0.041 0.2452 1>, Variant 2 Color=?)
      Float=&Multiply_Float(A=0.5, B=$Particles.AlphaScale)
     )
     Scale Factor=&MakeFloat4FromVectorAndFloat("Vector (RGB)"=$System.ToD, "Float (Alpha)"=1)
    )) {isSetParameters=true, module=SetVariables_#13}
   /SetVariables(
    Particles.DecalDyn:Vector4f=&MakeVector4(
     X=0.5
     Y=0.8
     Z=&RandomRangeFloat(Minimum=0, Maximum=1)
     W=&RandomRangeFloat(Minimum=0, Maximum=0)
    )
    Particles.FireDecalColor:LinearColor=&MultiplyLinearColor(
     Linear Color=&Multiply_LinearColorByFloat(
      Linear Color=&VariantColorSelector(Variant Amount=0, Original Color=<rgba 0.1721 0.7031 0.6058 1>, Variant 1 Color=<rgba 0.5505 0.2014 0.6042 1>, Variant 2 Color=?)
      Float=&Multiply_Float(A=1.75, B=$Particles.AlphaScale)
     )
     Scale Factor=&MakeFloat4FromVectorAndFloat("Vector (RGB)"=$System.ToD, "Float (Alpha)"=1)
    )
    Particles.FireDecalSize:Vector2f=&Multiply_Vector2D_ByFloat(Float=0.65, Vector2D=$Particles.SpriteSize)
    Particles.GlowDecalColor:LinearColor=&MultiplyLinearColor(
     Linear Color=&Multiply_LinearColorByFloat(
      Linear Color=&VariantColorSelector(Variant Amount=0, Original Color=<rgba 0.0054 0.195 0.184 1>, Variant 1 Color=<rgba 0.125 0.0182 0.1125 1>, Variant 2 Color=?)
      Float=&Multiply_Float(A=1.75, B=$Particles.AlphaScale)
     )
     Scale Factor=&MakeFloat4FromVectorAndFloat("Vector (RGB)"=$System.ToD, "Float (Alpha)"=1)
    )
   ) {isSetParameters=true, module=SetVariables_#14}
  ]
  rendererClasses [&NiagaraSpriteRendererProperties, &NiagaraMeshRendererProperties]
  renderers [
   {rendererClass=&NiagaraSpriteRendererProperties}
   {rendererClass=&NiagaraMeshRendererProperties, rendererIndex=1}
   {rendererClass=&NiagaraMeshRendererProperties, rendererIndex=2}
   {rendererClass=&NiagaraMeshRendererProperties, rendererIndex=3}
   {rendererClass=&NiagaraMeshRendererProperties, rendererIndex=4}
   {rendererClass=&NiagaraSpriteRendererProperties, rendererIndex=5}
   {rendererClass=&NiagaraSpriteRendererProperties, rendererIndex=6}
  ]
 }
}
summary {
 systemName=NS_Backpack_PolishedJade_Idle
 emitters [
  {
   emitterName=BlowingParticles
   rendererClasses [&NiagaraLightRendererProperties, &NiagaraMeshRendererProperties]
  }
  {
   emitterName=SimpleSpriteBurst001
   rendererClasses [&NiagaraSpriteRendererProperties, &NiagaraMeshRendererProperties]
  }
  {
   emitterName=SimpleSpriteBurst003
   rendererClasses [&NiagaraSpriteRendererProperties, &NiagaraMeshRendererProperties]
  }
  {emitterName=Flare001, rendererClasses [&NiagaraSpriteRendererProperties]}
  {emitterName=Flare2001, rendererClasses [&NiagaraSpriteRendererProperties]}
  {emitterName=FlareFE, rendererClasses [&NiagaraSpriteRendererProperties]}
  {emitterName=Flare2FE, rendererClasses [&NiagaraSpriteRendererProperties]}
 ]
 userVariables [
  User.bIsFrontEnd {type=NiagaraBool}
  User.bIsFrontEndPreview {type=NiagaraBool}
  User.Skeletal Mesh {
   type=NiagaraDataInterfaceSkeletalMesh
   defaultValue {
    struct=NiagaraExt_VariableValue_DataInterface
    value {dataInterface=&self:NiagaraDataInterfaceSkeletalMesh_0}
   }
  }
  User.Stage_01 {type=NiagaraBool, defaultValue=true}
  User.Stage_02 {type=NiagaraBool, defaultValue=true}
  User.Variant {type=NiagaraFloat, defaultValue=0, description=""}
 ]
}
