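# Material function for the MindGloom skin of the PolishedJade backpack, overriding BaseColor and EmissiveColor on incoming material attributes.
# Panned noise and UV-distorted smoke and emissive textures build glowing rings, flame, eyes, headband and pulsing lighting.
# Parameter-driven sphere masks in UV and pre-skinned space, including a second stage, gate the effects.
# Quality and shading-path switches drop some emissive layers on Low and mobile.
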MaterialFunction /BRCosmetics/Accessories/FORT_Backpacks/Backpack_F_MED_PolishedJade/Skins/MindGloom/Materials/MF_PolishedJade_MindGloom
graph {nodes [
  n1 = GetMaterialAttributes(None=FunctionInput() {bUsePreviewValueAsDefault=true, inputType=FunctionInput_MaterialAttributes}) {attributeGetTypes [#1, #2]}
  n2 = ScalarParameter() {parameterName=EffectMaskRadius}
  n3 = Multiply(
   A=NamedRerouteUsage()
   B=SphereMask(
    A=FunctionInput() {bUsePreviewValueAsDefault=true, inputName=PreSkinPosition}
    B=Constant3Vector() {constant=<rgba 94 0 0 0>}
    Radius=Multiply(A=@n2) {constB=1000}
   ) {attenuationRadius=100, hardnessPercent=90}
  )
  n4 = Constant()
  n5 = NamedRerouteUsage()
  n6 = Constant() {r=0.615}
  n7 = ComponentMask(None=TextureCoordinate()) {r=true}
  n8 = Add(
   A=NamedRerouteUsage()
   B=LinearInterpolate(B=NamedRerouteUsage(), Alpha=SmoothStep(Min=@n6, Max=@n6, Value=@n7) {constMax=0.75, constMin=0.75})
  )
  n9 = Add(A=NamedRerouteUsage(), B=NamedRerouteUsage())
  n10 = Add(
   A=ShadingPathSwitch(Default=QualitySwitch(Default=@n3, Low=@n4, Epic=@n3), Mobile=@n4)
   B=LinearInterpolate(A=@n1.EmissiveColor, B=Add(A=@n5, B=@n8), Alpha=Saturate(None=@n9))
  )
  n11 = NamedRerouteUsage()
  n12 = NamedRerouteUsage()
  FunctionOutput(None=SetMaterialAttributes!(MaterialAttributes=@n1.MaterialAttributes, EmissiveColor=@n10, BaseColor=LinearInterpolate(A=@n1.BaseColor, B=@n11, Alpha=@n12)) {attributeSetTypes [#1, #2]})
  MaterialFunctionCall(Emissive=@n10) {materialFunction=&MF_FX_EmissiveAdjustments}
  n13 = Constant() {r=0.65}
  n14 = TextureCoordinate()
  NamedRerouteDeclaration(None=SmoothStep(Min=@n13, Max=@n13, Value=ComponentMask(None=@n14) {g=true}) {constMax=0.75, constMin=0.75}) {name=Directional mask, nodeColor=<rgba 0 0.4609 1 1>}
  n15 = TextureCoordinate()
  n16 = AppendVector(A=ScalarParameter() {defaultValue=0.7, parameterName=U}, B=ScalarParameter() {defaultValue=0.75, parameterName=V})
  n17 = ScalarParameter() {parameterName=EffectMaskRadius_Stage2}
  NamedRerouteDeclaration(None=Multiply(A=SphereMask(A=@n15, B=@n16, Radius=@n17), B=ScalarParameter() {parameterName=stage2Blend})) {name=EffectMask stage2, nodeColor=<rgba 1 0 0.6797 1>}
  n18 = Multiply(A=NamedRerouteUsage(), B=Constant() {r=0.015}) {constB=0.05}
  n19 = TextureCoordinate()
  n20 = Constant2Vector() {r=-0.01}
  n21 = Constant2Vector() {r=0.005}
  n22 = NamedRerouteUsage()
  n23 = TextureSample(UVs=Add(A=Add(A=@n18, B=@n19), B=LinearInterpolate(A=@n20, B=@n21, Alpha=@n22))) {
   samplerSource=SSM_Wrap_WorldGroupSettings
   texture=/BRCosmetics/Accessories/FORT_Backpacks/Backpack_F_MED_PolishedJade/Skins/MindGloom/Textures/T_PolishedJadeMindGloom_Backpack_Smoke
  }
  n24 = Saturate(None=Multiply(A=Power(Base=@n23.G) {constExponent=1}) {constB=1.2})
  n25 = LinearInterpolate(Alpha=NamedRerouteUsage()) {constA=0.15}
  n26 = Multiply(A=Multiply(A=@n24, B=Power(Base=@n25)), B=NamedRerouteUsage())
  n27 = Multiply(A=NamedRerouteUsage()) {constB=0.001}
  n28 = LinearInterpolate(A=Constant2Vector(), B=Constant2Vector() {r=-0.001}, Alpha=NamedRerouteUsage())
  n29 = TextureSample(UVs=Add(A=Add(A=@n27, B=@n28), B=TextureCoordinate())) {
   samplerSource=SSM_Wrap_WorldGroupSettings
   texture=/BRCosmetics/Accessories/FORT_Backpacks/Backpack_F_MED_PolishedJade/Skins/MindGloom/Textures/T_PolishedJadeMindGloom_Backpack_E
  }
  n30 = Multiply(A=Saturate(None=Power(Base=@n29.G) {constExponent=1.5}), B=NamedRerouteUsage())
  NamedRerouteDeclaration(None=Multiply(A=Max(A=@n26, B=@n30)) {constB=2}) {name=Center circle color, nodeColor=<rgba 0 0.625 1 1>}
  n31 = VectorParameter() {parameterName=MIndGloom_Outer ring color, defaultValue=<rgba 0.1986 0.0276 0.207 1>}
  NamedRerouteDeclaration(None=LinearInterpolate(A=@n31.RGB, B=VectorParameter() {parameterName=MIndGloom_Inner ring color, defaultValue=<rgba 0.4 0.0388 0.3435 1>}.RGB, Alpha=@n24)) {name=Flame color, nodeColor=<rgba 0.9453 1 0 1>}
  n32 = ScalarParameter() {defaultValue=0.7, parameterName=U}
  n33 = ScalarParameter() {defaultValue=0.75, parameterName=V}
  n34 = ScalarParameter() {parameterName=EffectMaskRadius}
  NamedRerouteDeclaration(None=SphereMask(A=TextureCoordinate(), B=AppendVector(A=@n32, B=@n33), Radius=Max(A=@n34, B=@n17))) {name=EffectMask, nodeColor=<rgba 1 0 0.6797 1>}
  n35 = TextureSample() {texture=/BRCosmetics/Accessories/FORT_Backpacks/Backpack_F_MED_PolishedJade/Textures/T_PolishedJade_Backpack_Mask}
  n36 = VectorParameter() {parameterName=mindGloom_bodycolor, defaultValue=<rgba 3 0 1.3283 1>}
  n37 = Time()
  n38 = Saturate(None=Sine(None=Multiply(A=@n37) {constB=4}))
  NamedRerouteDeclaration(None=Multiply(A=Multiply(A=@n35.B, B=@n36.RGB), B=LinearInterpolate(Alpha=@n38) {constA=0.5, constB=0.55}) {constB=0.5}) {name=Lighting, nodeColor=<rgba 1 0.4688 0 1>}
  n39 = NamedRerouteUsage()
  n40 = NamedRerouteUsage()
  NamedRerouteDeclaration(None=Add(A=NamedRerouteUsage(), B=Add(A=@n39, B=@n40))) {name=EyesGlow, nodeColor=<rgba 0 0.3203 1 1>}
  n41 = TextureCoordinate()
  n42 = Constant2Vector() {g=0.4, r=0.2}
  n43 = Constant2Vector() {g=-0.4, r=0.2}
  n44 = NamedRerouteUsage()
  n45 = TextureSample(UVs=Panner(Coordinate=Multiply(A=@n41) {constB=2}, Speed=LinearInterpolate(A=@n42, B=@n43, Alpha=@n44) {constA=1, constB=-1}) {speedX=0.2, speedY=0.4}) {samplerSource=SSM_Wrap_WorldGroupSettings, samplerType=SAMPLERTYPE_Grayscale, texture=&LowResBlurredNoise}
  n46 = TextureCoordinate()
  n47 = Constant2Vector() {g=0.7, r=0.35}
  n48 = Constant2Vector() {g=-0.7, r=0.35}
  n49 = NamedRerouteUsage()
  n50 = TextureSample(UVs=Panner(Coordinate=Multiply(A=@n46) {constB=7}, Speed=LinearInterpolate(A=@n47, B=@n48, Alpha=@n49)) {speedX=0.35, speedY=0.7}) {samplerSource=SSM_Wrap_WorldGroupSettings, samplerType=SAMPLERTYPE_Grayscale, texture=&LowResBlurredNoise}
  NamedRerouteDeclaration(None=Max(A=@n45.R, B=@n50.R)) {name=Smoke, nodeColor=<rgba 0 1 0.1328 1>}
  n51 = TextureCoordinate()
  n52 = Constant2Vector() {g=0.6, r=0.3}
  n53 = Constant2Vector() {g=-0.6, r=0.3}
  n54 = NamedRerouteUsage()
  n55 = TextureSampleParameter2D(UVs=Panner(Coordinate=Multiply(A=@n51) {constB=6}, Speed=LinearInterpolate(A=@n52, B=@n53, Alpha=@n54)) {speedX=0.3, speedY=0.6}) {
   parameterName=Noise
   samplerSource=SSM_Wrap_WorldGroupSettings
   samplerType=SAMPLERTYPE_Grayscale
   texture=&LowResBlurredNoise
  }
  NamedRerouteDeclaration(None=@n55.R) {name=ColorNoise, nodeColor=<rgba 1 0 0.1875 1>}
  n56 = TextureCoordinate() {uTiling=0.5}
  n57 = TextureSample(UVs=Panner(Coordinate=Multiply(A=@n56) {constB=10}) {speedX=0.5}) {samplerSource=SSM_Wrap_WorldGroupSettings, samplerType=SAMPLERTYPE_Grayscale, texture=&LowResBlurredNoise}
  n58 = TextureSample() {
   samplerSource=SSM_Wrap_WorldGroupSettings
   texture=&T_PolishedJadeMindGloom_Backpack_E
  }
  n59 = Multiply(A=@n58.B) {constB=5}
  n60 = Multiply(A=Power(Base=@n57.R), B=Saturate(None=Power(Base=@n59)))
  n61 = Constant()
  n62 = QualitySwitch(Default=Saturate(None=Multiply(A=@n60) {constB=0.08}), Low=@n61)
  n63 = Power(Base=@n57.R) {constExponent=2.5}
  n64 = Constant() {r=0.01}
  n65 = TextureSample(UVs=Add(A=Multiply(A=@n57.R, B=@n64) {constB=0.1}, B=TextureCoordinate())) {
   samplerSource=SSM_Wrap_WorldGroupSettings
   texture=&T_PolishedJadeMindGloom_Backpack_E
  }
  n66 = Multiply(A=@n65.R) {constB=1.5}
  n67 = Add(A=Multiply(A=@n63, B=Constant3Vector() {constant=<rgba 0.08 0.0135 0.0469 1>}) {constB=0.04}, B=Multiply(A=@n63, B=Saturate(None=@n66)))
  n68 = Constant3Vector() {constant=<rgba 3 0.3187 1.9484 1>}
  n69 = Constant3Vector() {constant=<rgba 0.02 0.0112 0.0181 1>}
  n70 = TextureSample() {
   samplerSource=SSM_Wrap_WorldGroupSettings
   texture=&T_PolishedJadeMindGloom_Backpack_E
  }
  n71 = LinearInterpolate(
   A=Constant3Vector() {constant=<rgba 0.1 0.013 0.0754 1>}
   B=VectorParameter() {parameterName=MIndGloom_BaseColorA, defaultValue=<rgba 0.7379 0.0273 0.1356 1>}.RGB
   Alpha=Fresnel(Normal=PixelNormalWS()) {exponent=3}
  )
  n72 = Add(
   A=Add(A=ShadingPathSwitch(Default=@n62, Mobile=@n61), B=Multiply(A=@n67, B=@n68) {constB=3})
   B=Add(A=LinearInterpolate(B=@n69, Alpha=@n70.G), B=Power(Base=@n71) {constExponent=1.2})
  )
  n73 = TextureSample() {
   samplerSource=SSM_Wrap_WorldGroupSettings
   texture=&T_PolishedJade_Backpack_Mask
  }
  NamedRerouteDeclaration(None=LinearInterpolate(A=Constant(), B=Multiply(A=@n72) {constB=4}, Alpha=@n73.R)) {name=headband, nodeColor=<rgba 0.9219 1 0 1>}
  n74 = TextureCoordinate()
  n75 = TextureSample(UVs=Panner(Coordinate=Multiply(A=@n74) {constB=3}) {speedY=0.2}) {samplerSource=SSM_Wrap_WorldGroupSettings, samplerType=SAMPLERTYPE_Grayscale, texture=&LowResBlurredNoise}
  n76 = LinearInterpolate(Alpha=@n75.R) {constA=0.2}
  NamedRerouteDeclaration(None=Multiply(A=Multiply(A=@n76, B=@n35.G), B=@n36.RGB)) {name=Body, nodeColor=<rgba 1 0 0.7734 1>}
  n77 = TextureSample() {
   samplerSource=SSM_Wrap_WorldGroupSettings
   texture=/BRCosmetics/Accessories/FORT_Backpacks/Backpack_F_MED_PolishedJade/Skins/MindGloom/Textures/T_PolishedJade_MindGloom_Backpack_D2
  }
  n78 = TextureSample() {
   samplerSource=SSM_Wrap_WorldGroupSettings
   texture=&T_PolishedJadeMindGloom_Backpack_E
  }
  NamedRerouteDeclaration(None=LinearInterpolate(A=@n77.RGB, B=VectorParameter() {parameterName=MIndGloom_BaseColorA, defaultValue=<rgba 0.927 0 0.4329 1>}.RGB, Alpha=Saturate(None=@n78.B))) {name=BaseColor, nodeColor=<rgba 0.6641 0 1 1>}
  n79 = Multiply(A=NamedRerouteUsage()) {constB=0.002}
  n80 = TextureCoordinate()
  n81 = Constant2Vector() {r=-0.002}
  n82 = Constant2Vector() {r=-0.001}
  n83 = NamedRerouteUsage()
  n84 = TextureSample(UVs=Add(A=Add(A=@n79, B=@n80), B=LinearInterpolate(A=@n81, B=@n82, Alpha=@n83))) {
   samplerSource=SSM_Wrap_WorldGroupSettings
   texture=&T_PolishedJadeMindGloom_Backpack_E
  }
  n85 = Saturate(None=Multiply(A=@n84.R) {constB=0.75})
  n86 = Multiply(A=NamedRerouteUsage(), B=Constant() {r=0.03}) {constB=0.05}
  n87 = TextureCoordinate()
  n88 = Constant2Vector()
  n89 = Constant2Vector() {g=-0.01}
  n90 = NamedRerouteUsage()
  n91 = TextureSample(UVs=Add(A=Add(A=@n86, B=@n87), B=LinearInterpolate(A=@n88, B=@n89, Alpha=@n90))) {texture=&T_PolishedJadeMindGloom_Backpack_Smoke}
  n92 = Saturate(None=Multiply(A=Power(Base=@n91.R) {constExponent=3}) {constB=2})
  n93 = LinearInterpolate(A=VectorParameter() {parameterName=MIndGloom_Inner ring color, defaultValue=<rgba 0.4 0.0388 0.3435 1>}.RGB, Alpha=@n91.R)
  n94 = NamedRerouteUsage()
  n95 = Multiply(A=Multiply(A=Multiply(A=@n92, B=@n93)) {constB=10}, B=Power(Base=LinearInterpolate(Alpha=@n94)))
  NamedRerouteDeclaration(None=Multiply(A=Max(A=@n85, B=@n95))) {name=Inner ring color, nodeColor=<rgba 0.2188 1 0 1>}
  NamedRerouteDeclaration(None=Multiply(A=Multiply(A=@n23.B, B=@n31.RGB)) {constB=0.7}) {name=Outer ring color, nodeColor=<rgba 1 0 0.3047 1>}
 ]}
