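# Character parent material for the Polished Jade backpack cosmetic, built on material attributes with a shading model chosen by expression, for skeletal meshes.
# The shared 2023 CharacterParent PreFX function supplies pre-skinned local position to two backpack-specific functions, MF_PolishedJade and its MindGloom skin variant.
# A static switch, "Is MindGloom?", selects one variant's result and passes it into the shared PostFX function alongside other PreFX outputs.
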Material /BRCosmetics/Accessories/FORT_Backpacks/Backpack_F_MED_PolishedJade/Materials/M_CharacterParent_PolishedJade_Backpack
graph {
 properties {
  bUseMaterialAttributes=true
  bUsedWithMeshDeformer=true
  bUsedWithMorphTargets=true
  bUsedWithSkeletalMesh=true
  pixelDepthOffsetMode=PDOM_Legacy
  shadingModel=MSM_FromMaterialExpression
  usedShadingModels=Default Lit | Subsurface | Clear Coat | Subsurface Profile | Cloth
 }
 nodes [
  n1 = MaterialFunctionCall() {materialFunction=&MF_CharacterParent_2023_PreFX}
  MaterialFunctionCall(
   Material=StaticSwitchParameter(
    True=MaterialFunctionCall!(PreSkinPosition=@n1.Pre skinned Local Position, In=@n1.Pre skinned Local Position) {materialFunction=/BRCosmetics/Accessories/FORT_Backpacks/Backpack_F_MED_PolishedJade/Skins/MindGloom/Materials/MF_PolishedJade_MindGloom}.Result
    False=MaterialFunctionCall!(PreSkinPosition=@n1.Pre skinned Local Position, In=@n1.Pre skinned Local Position) {materialFunction=&/MF_PolishedJade}.Result
   ) {parameterName=Is MindGloom?}
   M=@n1.M
   Pre Skinned Local Position=@n1.M
   Distance Blend=@n1.M
   Fresnel=@n1.M
  ) {materialFunction=&MF_CharacterParent_2023_PostFX}
 ]
}
